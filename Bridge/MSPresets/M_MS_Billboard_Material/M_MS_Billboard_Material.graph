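# Masked two-sided foliage material for instanced billboard meshes, with switchable camera-facing rotation and grass wind.
# World position offset turns the mesh about its pivot toward the camera's horizontal position, compensating for foliage Z rotation, and adds vertex-green-weighted SimpleGrassWind.
# Albedo gets color overlay and SpeedTree variation; a packed ORT texture supplies opacity, roughness and translucency.
# Flattened normals blend with a view-basis-transformed copy, bypassing tangent space.
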Material /Bridge/MSPresets/M_MS_Billboard_Material/M_MS_Billboard_Material
graph {
 properties {
  bTangentSpaceNormal=false
  bUsedWithInstancedStaticMeshes=true
  blendMode=BLEND_Masked
  pixelDepthOffsetMode=PDOM_Legacy
  shadingModel=MSM_TwoSidedFoliage
  twoSided=true
  usedShadingModels=Two Sided Foliage
 }
 nodes [
  n1 = MaterialFunctionCall!(
   WindIntensity=ScalarParameter() {defaultValue=0.15, group=07 - Wind, parameterName=Wind Intensity, sortPriority=1}
   WindWeight=ScalarParameter() {defaultValue=0.5, group=07 - Wind, parameterName=Wind Height, sortPriority=2}
   WindSpeed=ScalarParameter() {defaultValue=0.5, group=07 - Wind, parameterName=Wind Speed, sortPriority=3}
   AdditionalWPO=ScalarParameter() {group=07 - Wind, parameterName=Additional WPO, sortPriority=4}
  ) {materialFunction=/Engine/Functions/Engine_MaterialFunctions01/WorldPositionOffset/SimpleGrassWind}
  n2 = Constant()
  n3 = VertexColor()
  n4 = ObjectOrientation()
  n5 = CameraPositionWS()
  n6 = Constant3Vector() {constant=<rgba 1 1 0 0>}
  n7 = Normalize(VectorInput=Subtract(A=ObjectPositionWS(), B=Multiply(A=@n5, B=@n6)))
  n8 = Divide(A=Arctangent2Fast(Y=ComponentMask(None=@n7) {g=true}, X=ComponentMask(None=@n7) {r=true})) {constB=3.1415}
  n9 = Subtract(
   A=Add(B=Divide(A=@n8)) {constA=0.25, constB=0.25}
   B=MaterialFunctionCall() {materialFunction=/Engine/ArtTools/RenderToTexture/MaterialFunctions/FoliageZRotation}."Z Rotation (0-1)"
  )
  n10 = MaterialFunctionCall() {materialFunction=&ObjectPivotPoint}
  n11 = WorldPosition()
  Add(
   A=StaticSwitchParameter(
    True=RotateAboutAxis(NormalizedRotationAxis=@n4, RotationAngle=@n9, PivotPoint=@n10.Object Pivot Location, Position=@n11.XYZ)
    False=Constant()
   ) {defaultValue=true, parameterName=Camera Facing Billboard}
   B=StaticSwitchParameter(True=LinearInterpolate(A=@n1.Result, B=@n2, Alpha=@n3.G), False=@n2) {group=07 - Wind, parameterName=EnableGrassWind}
  )
  n12 = Constant3Vector() {constant=<rgba 1 0 0 0>}
  n13 = Constant3Vector() {constant=<rgba 0 -1 0 0>}
  n14 = Constant3Vector() {constant=<rgba 0 0 -1 0>}
  n15 = TextureSampleParameter2D() {
   group=01- Input Textures
   parameterName=Normal
   samplerSource=SSM_Wrap_WorldGroupSettings
   samplerType=SAMPLERTYPE_Normal
   sortPriority=3
   texture=/Bridge/MSPresets/MSTextures/Placeholder_Normal
  }
  n16 = MaterialFunctionCall(Normal=@n15.RGB, Flatness=OneMinus(None=ScalarParameter() {defaultValue=1, group=05 - Normal, parameterName=Normal Intensity})) {materialFunction=&FlattenNormal}
  LinearInterpolate(
   A=@n16.Result
   B=MaterialFunctionCall(
    VectorToTransform=@n16.Result
    BasisX=Transform(None=@n12) {desc=RIGHT, transformSourceType=TRANSFORMSOURCE_View}
    BasisY=Transform(None=@n13) {desc=UP, transformSourceType=TRANSFORMSOURCE_View}
    BasisZ=Transform(None=@n14) {desc="FWD\r\n", transformSourceType=TRANSFORMSOURCE_View}
   ) {materialFunction=&Transform3x3Matrix}.Transformed Vector
  ) {constAlpha=3}
  n17 = TextureSampleParameter2D() {group=01- Input Textures, parameterName=Albedo, samplerSource=SSM_Wrap_WorldGroupSettings, texture=/Engine/EngineMaterials/DefaultDiffuse}
  n18 = MaterialFunctionCall(Base=@n17.RGB, Blend=VectorParameter() {group=02 - Albedo, parameterName=Color Overlay, defaultValue=<rgba 0.5 0.5 0.5 1>}.RGB) {materialFunction=&Blend_Overlay}
  n19 = ScalarParameter() {defaultValue=1, group=02 - Albedo, parameterName=Overlay Intensity, sliderMax=1.5, sortPriority=1}
  n20 = TextureSampleParameter2D() {desc=ORT, parameterName=ORT, samplerType=SAMPLERTYPE_LinearColor, texture=/Bridge/MSPresets/MSTextures/WhitePlaceholder}
  MaterialFunctionCall!(
   Albedo=MaterialFunctionCall!(
    BaseColor=LinearInterpolate(A=@n17.RGB, B=@n18.Result, Alpha=@n19)
    Amount=ScalarParameter() {group=02 - Albedo, parameterName=Color Variation, sortPriority=2}
   ) {desc=ms_BaseColor, materialFunction=/Engine/Functions/Engine_MaterialFunctions01/SpeedTree/SpeedTreeColorVariation}.Result
   Grayscale=@n20.B
  ) {materialFunction=/Bridge/MSPresets/M_MS_Foliage_Material/Functions/MF_Translucency}
  Multiply(A=@n20.G, B=ScalarParameter() {defaultValue=1, group=03- Roughness, parameterName=Roughness Intensity}) {desc=ms_Roughness}
  Multiply(A=@n20.R, B=ScalarParameter() {defaultValue=1, group=04 - Opacity, parameterName=Opacity Intensity}) {desc=ms_OpacityMask}
 ]
}
